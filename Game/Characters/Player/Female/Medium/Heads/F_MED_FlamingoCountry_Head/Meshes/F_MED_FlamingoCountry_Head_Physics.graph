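# Physics asset for the FlamingoCountry medium female player head, providing simulated motion for two dynamic hair bones.
# Each hair bone carries an offset sphere body, and the head bone carries a small capsule.
# Both hair bodies are constrained directly to the head rather than chained, with tight twist, narrow first swing and wider second swing limits.
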PhysicsAsset /Game/Characters/Player/Female/Medium/Heads/F_MED_FlamingoCountry_Head/Meshes/F_MED_FlamingoCountry_Head_Physics
bodies {
 dyn_hair_01 [{shapeType=Sphere, radius=5, shapeName=dyn_hair_01_sphere, center=<-3.4794 0 0>}]
 dyn_hair_02 [{shapeType=Sphere, radius=5, shapeName=dyn_hair_02_sphere, center=<4.5909 0 0>}]
 head [{shapeType=Capsule, length=2, radius=2, rotation=<rot 90 0 0>}]
}
constraints [
 {bone1Name=dyn_hair_02, bone2Name=head, swing1LimitDegrees=10, swing2LimitDegrees=60, twistLimitDegrees=2}
 {bone1Name=dyn_hair_01, bone2Name=head, swing1LimitDegrees=10, swing2LimitDegrees=60, twistLimitDegrees=2}
]
