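# Post-process Animation Blueprint for the FlamingoCountry female medium head, adding secondary physics motion to the incoming pose.
# A RigidBody node simulates the head physics asset in head-bone space, with damped component motion and overridden gravity, then feeds the head's Control Rig.
# The event graph sets Jump to 1 while jumping or falling, otherwise 0, but the AnimGraph shows its getter unconnected.
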AnimBlueprint /Game/Characters/Player/Female/Medium/Heads/F_MED_FlamingoCountry_Head/Meshes/F_Med_FlamingoCountry_Head_Post_Process
target_skeleton=&/F_MED_FlamingoCountry_Head_Skeleton
graphs [
 AnimGraph {nodes [
   n1 = RigidBody(ComponentPose=Local To Component(LocalPose=Input Pose())) {
    bClampLinearTranslationLimitToRefPose=true
    bOverrideWorldGravity=true
    overridePhysicsAsset=&/F_MED_FlamingoCountry_Head_Physics
    simulationSpace=BaseBoneSpace
    alphaBoolBlend {blendOption=Linear}
    baseBoneRef {boneName=head}
    componentAppliedLinearAccClamp=<2500 2500 2500>
    componentLinearAccScale=<0.3 0.2 0.2>
    componentLinearVelScale=<0.3 0.2 0.2>
    overrideWorldGravity {z=-980}
   }
   n2 = Control Rig(Source=Component To Local(ComponentPose=@n1)) {
    alphaBoolBlend {blendOption=Linear}
    controlRigAssetReference {blueprintRigClass=&/F_MED_FlamingoCountry_Head_CtrlRig.F_MED_FlamingoCountry_Head_CtrlRig_C}
   }
   Output Pose(Result=@n2) {layerGroup=None}
   Get Jump() -> (Jump:double)
  ]}
 EventGraph {nodes [
   Event BlueprintUpdateAnimation()
   n1 = Get bIsJumping()
   n2 = Get bIsFalling()
   > Set Jump(Jump:double=SelectFloat(A=1, B=0, bPickA=OR Boolean(A=@n1, B=@n2))) -> (Output_Get:double)
   TryGetPawnOwner()
  ]}
]
variables [Jump {type=number}]
